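# Parent anime-style material for the SpruceBark Bob character body, using material attributes with the Cloth shading model.
# It chains shared 2023 anime Pre, Calculations and Post functions, routing lit and shaded colors through a Bob-specific eye animation function masked by texture, vertex color or UV1.
# A sky-light half-vector specular highlight is added to both colors.
# A static IsKicks switch applies MF_Sock, and eye animation drives temporal responsiveness.
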Material /BRCosmetics/Characters/Player/Male/Medium/Bodies/M_MED_SpruceBark_Bob/Materials/M_SpruceBark_Bob_Anime_Parent
graph {
 properties {
  bAutomaticallySetUsageInEditor=false
  bUseMaterialAttributes=true
  bUsedWithClothing=true
  bUsedWithMeshDeformer=true
  bUsedWithMorphTargets=true
  bUsedWithNiagaraMeshParticles=true
  bUsedWithSkeletalMesh=true
  bUsedWithStaticLighting=true
  pixelDepthOffsetMode=PDOM_Legacy
  previewMesh=/BRCosmetics/Characters/Player/Female/Medium/Bodies/F_MED_PhoenixKick_Gain/Meshes/F_MED_PhoenixKick_Gain
  shadingModel=MSM_Cloth
  usedShadingModels=Cloth
 }
 nodes [
  n1 = Normalize(VectorInput=Add(A=CameraVectorWS(), B=SkyAtmosphereLightDirection()))
  n2 = MaterialFunctionCall!(CustomUV) {materialFunction=/Game/Global/Materials/Character_Parent_2023/Anime_Parent_2023/MaterialFunctions/MFC_AnimeShader_2023_Pre}
  n3 = Transform(None=@n2.Normal)
  n4 = Power(Base=Saturate(None=DotProduct(A=@n1, B=@n3)), Exp=Constant() {r=5}) {constExponent=6}
  n5 = Constant() {r=0.01}
  NamedRerouteDeclaration(None=Multiply(A=Multiply(A=@n4, B=@n5) {constB=0.01}, B=SkyAtmosphereLightIlluminance!(World Position))) {name=Specular, nodeColor=<rgba 0 1 0.2969 1>}
  n6 = MaterialFunctionCall!(CustomLightVector, OverrideNormal_WS, STT=@n2.STT, Normal_Map=@n2.STT, DFL=@n2.STT, FX=@n2.STT) {materialFunction=/Game/Global/Materials/Character_Parent_2023/Anime_Parent_2023/MaterialFunctions/MFC_AnimeShader_2023_Calculations}
  n7 = VertexColor()
  n8 = TextureSampleParameter2D() {
   group=01_EyeAnimation_Parameters
   parameterName=EyeMask
   samplerSource=SSM_Wrap_WorldGroupSettings
   samplerType=SAMPLERTYPE_Masks
   sortPriority=-32
   texture=&T_WhiteMask
  }
  n9 = StaticSwitchParameter(True=AppendVector(A=@n7, B=@n7), False=@n8.RGBA) {group=01_EyeAnimation_Parameters, parameterName=UseVertexColorForEyeMask, sortPriority=-33}
  n10 = TextureCoordinate() {coordinateIndex=1}
  n11 = Sign(None=Subtract(A=ComponentMask(None=@n10) {r=true}) {constB=0.0001})
  n12 = MaterialFunctionCall!(
   Color Lit=Multiply(A=@n2.Color_Lit, B=ScalarParameter() {defaultValue=2.3, parameterName=Lit Intensity})
   Color Shaded=Desaturation(None=@n2.Color_Shaded, Fraction=ScalarParameter() {defaultValue=-0.8})
   Use UV2=StaticBoolParameter() {defaultValue=true, group=01_EyeAnimation_Parameters, parameterName=Use UV2 for Eye Animation, sortPriority=-30}
   Mask=StaticSwitchParameter(
    True=ChannelMaskParameter(None=@n9) {group=01_EyeAnimation_Parameters, parameterName=EyeMaskChannel, sortPriority=-32}.R
    False=Saturate(None=@n11)
   ) {group=01_EyeAnimation_Parameters, parameterName=UseEyeMask, sortPriority=-33}
  ) {materialFunction=&/MF_AnimeEyeAnimation_SpruceBark_Bob}
  n13 = NamedRerouteUsage()
  n14 = MaterialFunctionCall!(
   Shading_Gradient=@n6.Shading_Gradient
   Threshold_Gradient=@n6.Shading_Gradient
   Specular_Gradient=@n6.Shading_Gradient
   Sheen_Mask=@n6.Shading_Gradient
   Thickness=@n6.Shading_Gradient
   DFL_Mask=@n6.Shading_Gradient
   FX=@n6.Shading_Gradient
   Metallic_Gradient=@n6.Shading_Gradient
   RimLight_Gradient=@n6.Shading_Gradient
   NormalBuffer=@n6.Shading_Gradient
   --------
   CINE_LightSources=@n6.Shading_Gradient
   CINE_LightColor=@n6.Shading_Gradient
   ---------
   Color_Lit=Add(A=@n12.Lit, B=@n13)
   Color_Shaded=Add(A=@n12.Shaded, B=@n13)
   SRM=@n2.SRM
   Emissive=@n2.SRM
   DFL_Color=@n2.SRM
   Specular_Color=@n2.SRM
   RimLight_Color=@n2.SRM
  ) {materialFunction=/Game/Global/Materials/Character_Parent_2023/Anime_Parent_2023/MaterialFunctions/MFC_AnimeShader_2023_Post}
  StaticSwitchParameter(
   True=MaterialFunctionCall!(InOpacity, Material=@n14.Result, Mask) {materialFunction=/Game/Global/Materials/Cosmetics/Kicks/Socks/Materials/MF_Sock}.Result
   False=@n14.Result
  ) {group=Kicks, parameterName=IsKicks}
  TemporalResponsivenessOutput!(None=@n12.TemporalResponsiveness)
 ]
}
