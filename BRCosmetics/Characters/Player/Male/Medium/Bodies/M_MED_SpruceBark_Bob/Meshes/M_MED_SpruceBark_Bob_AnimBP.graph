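# Post-process style animation Blueprint for the SpruceBark Bob player body, copying the attached parent mesh's pose and adjusting it through a Control Rig.
# Additive bone-space Modify Bone nodes on head and neck_01 follow the rig.
# Each update computes partial counter-rotations from those bones' ref-pose deltas, negated and scaled per axis, into rotator variables.
# The graph shows no link from those variables to the bone nodes, whose rotations read zero.
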AnimBlueprint /BRCosmetics/Characters/Player/Male/Medium/Bodies/M_MED_SpruceBark_Bob/Meshes/M_MED_SpruceBark_Bob_AnimBP
target_skeleton=&/M_MED_SpruceBark_Bob_Skeleton
graphs [
 AnimGraph {nodes [
   n1 = Control Rig(Source=Copy Pose From Mesh() {bUseAttachedParent=true}) {controlRigAssetReference {controlRigAsset=&/M_MED_SpruceBark_Bob_CtrlRig}}
   n2 = "Transform (Modify) Bone - Bone: head"(ComponentPose=Local To Component(LocalPose=@n1), Alpha=1, Translation="0,0,0", Rotation="0,0,0", Scale="1,1,1") {rotationMode=BMM_Additive, rotationSpace=BCS_BoneSpace, boneToModify {boneName=head}}
   Output Pose(Result=Component To Local(ComponentPose="Transform (Modify) Bone - Bone: neck_01"(ComponentPose=@n2, Alpha=1, Translation="0,0,0", Rotation="0,0,0", Scale="1,1,1") {rotationMode=BMM_Additive, rotationSpace=BCS_BoneSpace, boneToModify {boneName=neck_01}})) {layerGroup=None}
  ]}
 EventGraph {nodes [
   Event BlueprintUpdateAnimation()
   > n1 = Sequence()
   n2 = GetOwnerSkeletalMeshComponent()
   n3 = BreakRotator(InRot=BreakTransform(InTransform=GetDeltaTransformFromRefPose(self=@n2, BoneName=neck_01, BaseName=spine_05)).Rotation)
   n4 = Get Z percentage() -> (Z percentage:float)
   n5 = Get X percentage() -> (X percentage:float)
   n6 = Get Y percentage() -> (Y percentage:float)
   > Set neck_01_rot(neck_01_rot:Rotator=MakeRotator(
     Roll=float * float!(A=@n3, B=-1, C:double=@n5) -> (ReturnValue)
     Pitch=float * float(A=@n3.Pitch, B=-1, C=@n6)
     Yaw=float * float(A=@n3.Yaw, B=-1, C=@n4)
    )) -> (Output_Get:Rotator)
   n7 = GetOwnerSkeletalMeshComponent()
   n8 = BreakRotator(InRot=BreakTransform(InTransform=GetDeltaTransformFromRefPose(self=@n7, BoneName=head, BaseName=neck_02)).Rotation)
   n9 = Get Z percentage()
   n10 = Get X percentage()
   n11 = Get Y percentage()
   Set head_01_rot(
    execute=@n1.then_1
    head_01_rot:Rotator=MakeRotator(Roll=float * float(A=@n8, B=-1, C=@n10), Pitch=float * float(A=@n8.Pitch, B=-1, C=@n11), Yaw=float * float(A=@n8.Yaw, B=-1, C=@n9))
   ) -> (Output_Get:Rotator)
  ]}
]
variables [
 head_01_rot {type=Rotator}
 neck_01_rot {type=Rotator}
 X percentage {type=number, default=0.35}
 Y percentage {type=number, default=0.5}
 Z percentage {type=number, default=0.6}
]
